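# Unlit material for the Landmass erosion brush that computes eroded terrain height from particle simulation render targets, with negative emissive values allowed.
# SceneCaptureRT packs velocity, sediment transfer and water amount, while PositionRT feeds a Weathering-scaled offset.
# Named reroutes carry these values and advection offsets scaled by MPC_Erosion grid settings.
# HeightRT is sampled at offset, clamped texel-center UVs, then a term scaled by ApplySediment is added.
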Material /Landmass/Landscape/ErosionBrush/Materials/Particles/ParticleErosion_ErodeTerrain
graph {
 properties {bAllowNegativeEmissiveColor=true, pixelDepthOffsetMode=PDOM_Legacy, shadingModel=MSM_Unlit, usedShadingModels=Unlit}
 nodes [
  n1 = NamedRerouteUsage()
  n2 = NamedRerouteUsage()
  n3 = MaterialFunctionCall!(Add=Add(A=MaterialFunctionCall() {materialFunction=/Landmass/Landscape/OldPrototype_BP/Mats/MF/PointSampleUVs}.TexelCenterUVs, B=Add(A=@n1, B=@n2))) {materialFunction=/Landmass/Landscape/ErosionBrush/Materials/Functions/ErosionBrush_ClampUVs}
  n4 = Constant()
  n5 = NamedRerouteUsage()
  n6 = ScalarParameter() {defaultValue=1, parameterName=ApplySediment}
  MaterialFunctionCall(X=Add(
    A=TextureSampleParameter2D!(UVs=@n3.Clamp, Level=@n4, Apply View MipBias) {mipValueMode=TMVM_MipLevel, parameterName=HeightRT, texture=&Black}.R
    B=Multiply(A=@n5, B=@n6)
   )) {materialFunction=&MakeFloat3}
  n7 = MaterialFunctionCall() {materialFunction=&PointSampleUVs}
  n8 = Constant()
  n9 = CollectionParameter() {collection=/Landmass/Landscape/BlueprintBrushes/MPC/MPC_Erosion, parameterName=ResX}
  n10 = CollectionParameter() {collection=&MPC_Erosion, parameterName=ResY}
  n11 = Divide(
   A=ComponentMask(None=TextureSampleParameter2D(UVs=@n7.TexelCenterUVs, Level=@n8) {mipValueMode=TMVM_MipLevel, parameterName=PositionRT, texture=&Black}.RGB) {g=true, r=true}
   B=Multiply(
    A=AppendVector(A=@n9, B=@n10)
    B=CollectionParameter() {collection=&MPC_Erosion, parameterName=GridSize}
   )
  )
  NamedRerouteDeclaration(None=Multiply(A=Multiply(A=@n11) {constB=0.01}, B=ScalarParameter() {parameterName=Weathering})) {name=Position Advection, nodeColor=<rgba 1 0 0.4219 1>}
  n12 = MaterialFunctionCall(Vector 2=NamedRerouteUsage()) {materialFunction=/Engine/Functions/Engine_MaterialFunctions02/Utility/VectorLength}
  n13 = CollectionParameter() {collection=&MPC_Erosion, parameterName=GridSize}
  n14 = Divide(
   A=Min(B=Divide(A=@n12.V2 Length, B=@n13)) {constA=1}
   B=AppendVector(
    A=CollectionParameter() {collection=&MPC_Erosion, parameterName=ResX}
    B=CollectionParameter() {collection=&MPC_Erosion, parameterName=ResY}
   )
  )
  n15 = Normalize(VectorInput=Add(A=NamedRerouteUsage(), B=Constant2Vector() {r=0}))
  NamedRerouteDeclaration(None=Multiply(
    A=Multiply(A=@n14, B=@n15)
    B=CollectionParameter() {collection=&MPC_Erosion, parameterName=AdvectionErosion}
   )) {name=Velocity Advection, nodeColor=<rgba 0 1 0.8828 1>}
  n16 = TextureSampleParameter2D(UVs=MaterialFunctionCall() {materialFunction=&PointSampleUVs}.TexelCenterUVs, Level=Constant()) {mipValueMode=TMVM_MipLevel, parameterName=SceneCaptureRT, texture=&Black}
  NamedRerouteDeclaration(None=ComponentMask(None=@n16.RGB) {g=true, r=true}) {name=Velocity, nodeColor=<rgba 1 0.8091 0.0112 1>}
  NamedRerouteDeclaration(None=Multiply(A=Multiply(A=@n16.A) {constB=100}) {constB=1000}) {name=Water Amount, nodeColor=<rgba 0.0201 0.2862 1 1>}
  NamedRerouteDeclaration(None=@n16.B) {name=Sediment Transfer, nodeColor=<rgba 0.4062 0 1 1>}
  TextureSampleParameter2D(UVs=@n7.TexelCenterUVs, Level=@n8) {mipValueMode=TMVM_MipLevel, parameterName=VelocityRT, texture=&Black}
 ]
}
