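# Additive unlit preview material for Landmass erosion brush particles, drawn on instanced static meshes.
# Looks up particle position, water amount and velocity from render targets, scaled by MPC_Erosion values.
# World-position math builds camera-facing quads stretched along a motion axis, with an optional time-looped velocity offset.
# Velocity-based color and DebugFloat4Values readouts provide visual debugging.
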Material /Landmass/Landscape/ErosionBrush/Materials/Particles/ParticleErosion_ParticlePreview
graph {
 properties {
  bTangentSpaceNormal=false
  bUsedWithInstancedStaticMeshes=true
  blendMode=BLEND_Additive
  pixelDepthOffsetMode=PDOM_Legacy
  shadingModel=MSM_Unlit
  twoSided=true
  usedShadingModels=Unlit
 }
 nodes [
  n1 = VectorParameter() {parameterName=Emissive, defaultValue=<rgba 0 0 1 1>}
  n2 = ScalarParameter() {defaultValue=1, parameterName=Blue Emissive}
  n3 = ComponentMask(None=NamedRerouteUsage()) {g=true, r=true}
  n4 = CollectionParameter() {collection=/Landmass/Landscape/BlueprintBrushes/MPC/MPC_Erosion, parameterName=deltaT}
  n5 = Divide(
   A=Length(None=Multiply(A=@n3, B=@n4)) {is=class MaterialExpressionLength}
   B=ScalarParameter() {defaultValue=8, parameterName=Velocity Color Threshold}
  )
  Add(A=NamedRerouteUsage(), B=LinearInterpolate(A=Multiply(A=@n1.RGB, B=@n2), B=NamedRerouteUsage(), Alpha=Saturate(None=@n5)))
  Constant() {r=1}
  n6 = MaterialFunctionCall() {materialFunction=/Landmass/Landscape/ErosionBrush/Materials/Functions/ErosionBrush_ParticleCoords}
  n7 = Subtract(
   A=TextureSampleParameter2D(UVs=@n6.Particle UVs) {parameterName=VelocityRT, texture=&Black}.A
   B=TextureSampleParameter2D(UVs=@n6.Particle UVs) {parameterName=VelocityRTPrev, texture=&Black}.A
  ) {desc="Positive = Saturation Increase"}
  n8 = Saturate(None=Ceil(None=@n7))
  LinearInterpolate(A=Constant3Vector() {constant=<rgba 0.3 0 0 0>}, B=Constant3Vector() {constant=<rgba 0 0.3 0 0>}, Alpha=@n8)
  n9 = NamedRerouteUsage()
  Multiply(A=Saturate(None=Multiply(A=@n9) {constB=1000}), B=ScalarParameter() {defaultValue=1, parameterName=Opacity}) {constB=0.5}
  Multiply(B=@n8)
  n10 = Multiply(A=NamedRerouteUsage()) {constB=1000}
  Multiply(A=Multiply(A=@n7, B=@n10))
  n11 = NamedRerouteUsage()
  n12 = NamedRerouteUsage()
  n13 = ScalarParameter() {defaultValue=1, parameterName=Debug Position}
  n14 = VectorParameter() {defaultValue=<rgba 0.3 0 0.7 0.9>}
  n15 = MaterialFunctionCall!(
   Vector4=AppendVector(A=LinearInterpolate(A=@n11, B=@n12, Alpha=@n13), B=@n10)
   MaximumNumberOfDigits
   UVs=TextureCoordinate()
   DebugTextLocation RG_UpperRight BA_LowerLeft=AppendVector(A=@n14.RGB, B=@n14.RGB)
   Component Spacing
  ) {materialFunction=/Engine/Functions/Engine_MaterialFunctions02/Utility/DebugFloat4Values}
  n16 = ScalarParameter() {defaultValue=1, parameterName=Show Debug}
  NamedRerouteDeclaration(None=ComponentMask(None=Multiply(A=@n15.ColorCodedOutput, B=@n16)) {b=true, g=true, r=true}) {name=Debug Text, nodeColor=<rgba 1 0 0.9609 1>}
  n17 = Multiply(A=NamedRerouteUsage(), B=CollectionParameter() {collection=&MPC_Erosion, parameterName=deltaT})
  NamedRerouteDeclaration(None=Multiply(A=Abs(None=@n17), B=ScalarParameter() {defaultValue=0.01, parameterName=Velocity Brightness})) {name=Velocity Color, nodeColor=<rgba 0 1 0.9531 1>}
  n18 = CollectionParameter() {collection=&MPC_Erosion, parameterName=LandscapeLocation}
  NamedRerouteDeclaration(None=Add(A=ComponentMask(None=@n18) {b=true}, B=ScalarParameter() {parameterName=Z Offset})) {name=Z Offset, nodeColor=<rgba 0.1073 0.3797 1 1>}
  NamedRerouteDeclaration(None=TextureSampleParameter2D(UVs=MaterialFunctionCall() {materialFunction=&ErosionBrush_ParticleCoords}.Particle UVs) {parameterName=VelocityRT, texture=&Black}.RGB) {name=Particle Velocity, nodeColor=<rgba 1 0.9485 0 1>}
  n19 = AppendVector(
   A=MaterialFunctionCall() {materialFunction=&ErosionBrush_ParticleCoords}.Particle Position XY
   B=NamedRerouteUsage()
  )
  n20 = NamedRerouteUsage()
  NamedRerouteDeclaration(None=Add(A=Add(A=@n19, B=@n20), B=NamedRerouteUsage())) {name=Particle World XYZ, nodeColor=<rgba 0.7344 1 0 1>}
  n21 = Divide(
   A=AppendVector(
    A=CollectionParameter() {collection=&MPC_Erosion, parameterName=ResX}
    B=CollectionParameter() {collection=&MPC_Erosion, parameterName=ResY}
   )
   B=CollectionParameter() {collection=&MPC_Erosion, parameterName=ResolutionDivisor}
  )
  n22 = Floor(None=Multiply(
    A=MaterialFunctionCall() {materialFunction=&ErosionBrush_ParticleCoords}.Particle UVs
    B=@n21
   ))
  n23 = TextureSampleParameter2D(UVs=Divide(A=Add(A=@n22) {constB=0.5}, B=@n21)) {parameterName=PositionRT, texture=&Black}
  NamedRerouteDeclaration(None=@n23.A) {name=Water Amount, nodeColor=<rgba 0.1054 0.5861 1 1>}
  n24 = Multiply(
   A=CollectionParameter() {collection=&MPC_Erosion, parameterName=GridSize}
   B=ScalarParameter() {defaultValue=0.5, parameterName=Particle Drawscale}
  )
  n25 = Saturate(None=Ceil(None=NamedRerouteUsage()))
  NamedRerouteDeclaration(None=Multiply(
    A=Multiply(A=@n24, B=@n25)
    B=CollectionParameter() {collection=&MPC_Erosion, parameterName=ResolutionDivisor}
   )) {name=Particle Size, nodeColor=<rgba 0 1 0.1562 1>}
  n26 = ComponentMask(None=@n23.RGB) {g=true, r=true}
  n27 = Multiply(A=@n23.B, B=CollectionParameter() {collection=&MPC_Erosion, parameterName=ZScale})
  NamedRerouteDeclaration(None=Multiply(A=AppendVector(A=@n26, B=@n27)) {constB=100}) {name=Simulated Position, nodeColor=<rgba 0 1 0.3906 1>}
  n28 = Multiply(A=CollectionParameter() {collection=&MPC_Erosion, parameterName=deltaT}, B=NamedRerouteUsage())
  n29 = Frac(None=Time())
  NamedRerouteDeclaration(None=Multiply(A=Multiply(A=@n28, B=@n29), B=ScalarParameter() {parameterName=Show Delta})) {name=Debug Velocity Offset, nodeColor=<rgba 1 0 0.8203 1>}
  n30 = NamedRerouteUsage()
  n31 = ScalarParameter() {defaultValue=0.1, parameterName=MV Amount}
  n32 = Add(A=Length(None=Multiply(A=@n30, B=@n31)), B=NamedRerouteUsage())
  n33 = VectorParameter() {desc=Allows deriving normal when it's purely up Z., parameterName=v_epsilon, defaultValue=<rgba 0.001 0 0 0>}
  n34 = Normalize(VectorInput=Add(A=NamedRerouteUsage(), B=@n33.RGB)) {desc=X Axis}
  n35 = ConstantBiasScale(None=TextureCoordinate()) {bias=-0.5, scale=-1}
  n36 = Subtract(A=NamedRerouteUsage(), B=CameraPositionWS())
  n37 = Normalize(VectorInput=CrossProduct(A=@n34, B=Normalize(VectorInput=@n36))) {desc=Y Axis}
  n38 = NamedRerouteUsage()
  n39 = Add(
   A=Multiply(A=Multiply(A=@n32, B=@n34), B=ComponentMask(None=@n35) {r=true}) {desc=X}
   B=Multiply(A=ComponentMask(None=@n35) {g=true}, B=Multiply(A=@n37, B=@n38) {desc=Y})
  )
  n40 = NamedRerouteUsage()
  Reroute(None=Subtract(A=Add(A=@n39, B=@n40), B=WorldPosition() {worldPositionShaderOffset=WPT_ExcludeAllShaderOffsets}.XYZ))
  TextureCoordinate()
 ]
}
