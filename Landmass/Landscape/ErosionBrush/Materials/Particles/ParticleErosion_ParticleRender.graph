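# Unlit alpha-composite material for the Landmass erosion brush that renders instanced simulation particles as quads carrying velocity and sediment transfer.
# Per-instance custom data and UV1 derive texture coordinates into position and velocity render targets, with world scaling from MPC_Erosion.
# A World Position Offset chain optionally snaps particles to cell centers as 2x2 quads with bilinear opacity.
# Sediment transfer combines current-versus-previous velocity alpha with evaporation, and opacity is scaled down to capture overlapping particles.
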Material /Landmass/Landscape/ErosionBrush/Materials/Particles/ParticleErosion_ParticleRender
graph {
 properties {
  bAllowNegativeEmissiveColor=true
  bDisableDepthTest=true
  bTangentSpaceNormal=false
  bUsedWithInstancedStaticMeshes=true
  blendMode=BLEND_AlphaComposite
  pixelDepthOffsetMode=PDOM_Legacy
  shadingModel=MSM_Unlit
  twoSided=true
  usedShadingModels=Unlit
 }
 nodes [
  n1 = NamedRerouteUsage()
  n2 = VertexInterpolator(VS=@n1)
  n3 = TextureSampleParameter2D(UVs=NamedRerouteDeclaration(None=ShaderStageSwitch(PixelShader=@n2.PS, VertexShader=@n1)) {name=UV VS, nodeColor=<rgba 0 1 0.0859 1>}) {parameterName=PositionRT, texture=&Black}
  n4 = Multiply(A=@n3.RGB) {constB=100}
  Add(A=@n4, B=VectorParameter() {parameterName=manual offset}.RGB)
  n5 = CollectionParameter() {collection=/Landmass/Landscape/BlueprintBrushes/MPC/MPC_Erosion, parameterName=LandscapeLocation}
  AppendVector(
   A=MaterialFunctionCall() {materialFunction=/Landmass/Landscape/ErosionBrush/Materials/Functions/ErosionBrush_ParticleCoords}.Particle Position XY
   B=Add(A=ComponentMask(None=@n5) {b=true}, B=ScalarParameter() {parameterName=Zoffset})
  )
  n6 = TextureSampleParameter2D(UVs=NamedRerouteUsage()) {parameterName=PositionRT, texture=&Black}
  ComponentMask(None=Multiply(A=@n6.RGB) {constB=100}) {b=true}
  Constant() {r=2}
  Constant()
  n7 = ComponentMask(None=NamedRerouteUsage()) {g=true, r=true}
  n8 = NamedRerouteUsage()
  n9 = Floor(None=Divide(A=@n7, B=@n8))
  n10 = StaticSwitchParameter(True=Add(A=Multiply(A=@n9, B=@n8), B=Multiply(A=@n8) {constB=0.5}), False=@n7) {defaultValue=true, parameterName=Render AntiAliased 2x2 Quad}
  n11 = Multiply(
   A=ComponentMask(None=NamedRerouteUsage()) {b=true}
   B=CollectionParameter() {collection=&MPC_Erosion, parameterName=ZScale}
  )
  n12 = ComponentMask(None=ConstantBiasScale(None=TextureCoordinate()) {bias=-0.5, scale=1}) {g=true, r=true}
  n13 = StaticSwitchParameter(True=NamedRerouteUsage(), False=Constant()) {defaultValue=true, parameterName=Render AntiAliased 2x2 Quad}
  n14 = Add(A=Add(A=AppendVector(A=@n10, B=@n11)) {constB=0, desc=Not Used anymore}, B=AppendVector(A=Multiply(A=@n12, B=@n13), B=Constant()))
  LinearInterpolate(
   A=@n14
   B=Subtract(A=@n14, B=WorldPosition() {worldPositionShaderOffset=WPT_ExcludeAllShaderOffsets}.XYZ)
   Alpha=ScalarParameter() {parameterName=Center Verts out}
  )
  n15 = NamedRerouteUsage()
  LinearInterpolate(
   A=Multiply(A=WorldPosition() {worldPositionShaderOffset=WPT_ExcludeAllShaderOffsets}.XYZ) {constB=-1}
   B=@n14
   Alpha=Saturate(None=Ceil(None=@n15))
  )
  n16 = CollectionParameter() {collection=&MPC_Erosion, parameterName=RTWorldSizeX}
  n17 = CollectionParameter() {collection=&MPC_Erosion, parameterName=RTWorldSizeY}
  n18 = Add(
   A=Multiply(A=NamedRerouteUsage(), B=AppendVector(A=@n16, B=@n17))
   B=ComponentMask(None=CollectionParameter() {collection=&MPC_Erosion, parameterName=LandscapeLocation}) {g=true, r=true}
  )
  n19 = Divide(A=CollectionParameter() {collection=&MPC_Erosion, parameterName=GridSize})
  LinearInterpolate(
   A=Multiply(
    A=CollectionParameter() {collection=&MPC_Erosion, parameterName=GridSize}
    B=ScalarParameter() {parameterName=ShiftManualTest}
   )
   B=AppendVector(A=Subtract(A=@n18, B=@n19), B=ScalarParameter() {parameterName=Zoffset})
   Alpha=ScalarParameter() {parameterName=Center Verts out}
  )
  MaterialFunctionCall() {materialFunction=&ErosionBrush_ParticleCoords}
  MaterialFunctionCall() {materialFunction=&ErosionBrush_ParticleCoords}
  MaterialFunctionCall() {materialFunction=/Landmass/Landscape/OldPrototype_BP/Mats/MF/Scatter_InstanceCoords}
  MaterialFunctionCall() {materialFunction=&Scatter_InstanceCoords}
  n20 = PerInstanceCustomData!(DefaultValue)
  n21 = PerInstanceCustomData() {dataIndex=1}
  MaterialFunctionCall!(
   Vector2=AppendVector(A=ComponentMask(None=@n20) {r=true}, B=ComponentMask(None=@n21) {r=true})
   MaximumNumberOfDigits
   UVs
   DebugTextLocation RG_UpperRight BA_LowerLeft
   Component Spacing
  ) {materialFunction=/Engine/Functions/Engine_MaterialFunctions02/Utility/DebugFloat2Values}
  n22 = MaterialFunctionCall() {materialFunction=/Engine/Functions/Engine_MaterialFunctions02/Utility/VectorLength}
  n23 = CollectionParameter() {collection=&MPC_Erosion, parameterName=ResX}
  Multiply(
   A=CollectionParameter() {collection=&MPC_Erosion, parameterName=Solubility}
   B=Divide(
    A=Divide(A=@n22.V2 Length, B=@n23)
    B=CollectionParameter() {collection=&MPC_Erosion, parameterName=GridSize}
   )
  )
  Multiply(A=VectorParameter() {parameterName=DebugColor, defaultValue=<rgba 1 0 0 0>}.RGB, B=NamedRerouteUsage())
  NamedRerouteDeclaration(None=Multiply(A=NamedRerouteDeclaration(None=@n3.A) {name=Water Amount Normalized, nodeColor=<rgba 0 0.9766 1 1>}) {constB=1000}) {name=Water Amount, nodeColor=<rgba 0.0084 0.4764 1 1>}
  n24 = CollectionParameter() {collection=&MPC_Erosion, desc=Scales according to Downsampling, parameterName=ParticleSize}
  n25 = Divide(
   A=ComponentMask(None=NamedRerouteUsage()) {g=true, r=true}
   B=NamedRerouteDeclaration(None=@n24) {name=Particle Size, nodeColor=<rgba 0 0.6953 1 1>}
  )
  n26 = Frac(None=@n25)
  n27 = TextureCoordinate()
  n28 = Ceil(None=Subtract(A=OneMinus(None=@n27)) {constB=0.5})
  n29 = MaterialFunctionCall(Float2=LinearInterpolate(A=@n26, B=OneMinus(None=@n26), Alpha=Saturate(None=@n28))) {materialFunction=&BreakOutFloat2Components}
  n30 = Multiply(A=@n29.R, B=@n29.R)
  NamedRerouteDeclaration(None=StaticSwitchParameter(True=Saturate(None=@n30), False=Constant() {r=1}) {defaultValue=true, parameterName=Render AntiAliased 2x2 Quad}) {name=Bilinear Opacity, nodeColor=<rgba 0 0.1797 1 1>}
  NamedRerouteDeclaration(None=@n6.A) {name=WaterVS, nodeColor=<rgba 0.8047 1 0 1>}
  n31 = Multiply(
   A=TextureCoordinate() {coordinateIndex=1}
   B=CollectionParameter() {collection=&MPC_Erosion, parameterName=ComponentRes}
  )
  n32 = AppendVector(
   A=ComponentMask(None=PerInstanceCustomData() {desc=Vertex Offset X}) {r=true}
   B=ComponentMask(None=PerInstanceCustomData() {dataIndex=1, desc=Vertex Offset Y}) {r=true}
  )
  n33 = CollectionParameter() {collection=&MPC_Erosion, parameterName=SimulatedResolutionX}
  n34 = CollectionParameter() {collection=&MPC_Erosion, parameterName=SimulatedResolutionY}
  NamedRerouteDeclaration(None=Divide(A=Add(A=@n31, B=@n32), B=AppendVector(A=@n33, B=@n34)) {constB=8}) {name=CustomDataDerivedUVs, nodeColor=<rgba 1 0 0.375 1>}
  n35 = NamedRerouteUsage()
  n36 = TextureSampleParameter2D(UVs=@n35) {parameterName=VelocityRTPrev, texture=&Black}
  n37 = Multiply(
   A=Subtract(A=TextureSampleParameter2D(UVs=@n35) {parameterName=VelocityRT, texture=&Black}.A, B=@n36.A) {desc="Positive = Saturation Increase"}
   B=NamedRerouteUsage()
  )
  n38 = Multiply(
   A=CollectionParameter() {collection=&MPC_Erosion, parameterName=Evaporation}
   B=CollectionParameter() {collection=&MPC_Erosion, parameterName=deltaT}
  )
  n39 = Add(
   A=Multiply(A=Multiply(A=@n37) {constB=-1}, B=ScalarParameter() {defaultValue=1, parameterName=SedimentMultiplier})
   B=Multiply(A=Multiply(A=@n36.A, B=@n38), B=NamedRerouteUsage())
  )
  NamedRerouteDeclaration(None=@n39) {name=Sediment Transfer, nodeColor=<rgba 0 0.5078 1 1>}
  NamedRerouteDeclaration(None=@n4) {name=Position Offset, nodeColor=<rgba 0.1953 1 0 1>}
  n40 = TextureSampleParameter2D(UVs=NamedRerouteUsage()) {parameterName=VelocityRT, texture=&Black}
  n41 = ScalarParameter() {defaultValue=256, parameterName=Divisor}
  n42 = ComponentMask(None=Multiply(A=Divide(A=@n40.RGB, B=@n41), B=ScalarParameter() {defaultValue=1, parameterName=Output})) {g=true, r=true}
  n43 = NamedRerouteUsage()
  NamedRerouteDeclaration(None=Multiply(A=AppendVector(A=@n42, B=@n43), B=NamedRerouteUsage())) {name=Velocity And Sediment, nodeColor=<rgba 1 0.7266 0 1>}
  n44 = NamedRerouteUsage()
  NamedRerouteDeclaration(None=Multiply(A=Divide(A=@n44) {constB=100, desc=Allows capturing volume of up to 100 overlapping particles}, B=NamedRerouteUsage())) {name=Output Opacity, nodeColor=<rgba 0.0273 0.2986 1 1>}
  NamedRerouteDeclaration(None=@n14) {name=WPO, nodeColor=<rgba 0.3828 0 1 1>}
  NamedRerouteUsage()
  NamedRerouteUsage()
  NamedRerouteUsage()
  NamedRerouteUsage()
  NamedRerouteUsage()
  ScalarParameter() {defaultValue=64, desc=water amt, parameterName=n1}
  ScalarParameter() {defaultValue=-1, parameterName=EEE}
  VertexInterpolator(VS=@n42)
  VertexInterpolator(VS=@n25)
  VertexInterpolator(VS=@n39)
 ]
}
