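# Unlit material for the Landmass erosion brush that advances particle positions stored in render targets by one time step.
# A Custom HLSL node adds velocity times deltaT to each texel's UV-relative offset, converting meters to centimeters and back.
# It samples terrain height at the destination to keep particles above ground.
# Further nodes derive a border mask and alpha-based water volume with evaporation, using MPC_Erosion settings; negative emissive is allowed.
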Material /Landmass/Landscape/ErosionBrush/Materials/Particles/ParticleErosion_ScatterPosition
graph {
 properties {
  bAllowNegativeEmissiveColor=true
  blendMode=BLEND_AlphaComposite
  pixelDepthOffsetMode=PDOM_Legacy
  shadingModel=MSM_Unlit
  usedShadingModels=Unlit
 }
 nodes [
  MaterialExpressionTextureObjectParameter_1 = TextureObjectParameter() {parameterName=PositionRT, texture=&Black}
  MaterialExpressionTextureObjectParameter_5 = TextureObjectParameter() {parameterName=VelocityRT, texture=&Black}
  MaterialExpressionTextureObjectParameter_3 = TextureObjectParameter() {parameterName=NormalRT, texture=&Black}
  MaterialExpressionTextureObjectParameter_2 = TextureObjectParameter() {parameterName=HeightRT, texture=&Black}
  MaterialExpressionCollectionParameter_15 = CollectionParameter() {collection=/Landmass/Landscape/BlueprintBrushes/MPC/MPC_Erosion, parameterName=deltaT}
  MaterialExpressionAppendVector_4 = AppendVector(
   A=CollectionParameter() {collection=&MPC_Erosion, parameterName=RTWorldSizeX}
   B=CollectionParameter() {collection=&MPC_Erosion, parameterName=RTWorldSizeY}
  )
  MaterialExpressionTextureCoordinate_1 = TextureCoordinate()
  MaterialExpressionCollectionParameter_5 = CollectionParameter() {collection=&MPC_Erosion, parameterName=GridSize}
  MaterialExpressionCollectionParameter_2 = CollectionParameter() {collection=&MPC_Erosion, parameterName=ZScale}
  n1 = Custom!(
   PositionRT=@MaterialExpressionTextureObjectParameter_1
   VelocityRT=@MaterialExpressionTextureObjectParameter_5
   NormalRT=@MaterialExpressionTextureObjectParameter_3
   HeightRT=@MaterialExpressionTextureObjectParameter_2
   deltaT=@MaterialExpressionCollectionParameter_15
   TotalSize=@MaterialExpressionAppendVector_4
   inUV=@MaterialExpressionTextureCoordinate_1
   GridSize=@MaterialExpressionCollectionParameter_5
   ZScale=@MaterialExpressionCollectionParameter_2
  ) {
   code="//PositionRT: input Texture Object containing \r\n//VelcotiyRT: input \r\n//NormalRT: input \r\n//deltaT: input\r\n//input TextureSize\r\n//input GridSize\r\n\r\n//Sample position and velocity for each particle\r\nfloat4 position = PositionRT.SampleLevel(PositionRTSampler, inUV, 0).rgba;\r\nfloat3 velocity = VelocityRT.SampleLevel(VelocityRTSampler, inUV, 0).rgb;\r\n\r\n//Note: position currently converted from meters to cm due to clamping issues in RenderTargets\r\nposition.rgb *= 100;\r\n\r\nvelocity.z /= ZScale;\r\n\r\n//Add velocity to the position\r\nposition.rgb += (velocity * deltaT);\r\n\r\n//Get the location of the Terrain UV by adding the position as UV\r\nfloat2 TerrainUV = inUV + ((position.rg) / TotalSize);\r\n\r\n//Get the terrain height at the particle destination\r\nfloat endZ = HeightRT.SampleLevel(NormalRTSampler, TerrainUV, 0).r;\r\n\r\n//Prevent Particles from falling through Terrain\r\nposition.z = max(position.z, endZ + GridSize*0.25);\r\n\r\nposition.rgb /= 100;\r\n\r\nreturn position;"
   description=Erosion Pass
   outputType=CMOT_Float4
   inputs [
    {
     inputName=PositionRT
     input {expression=&self:MaterialExpressionTextureObjectParameter_1}
    }
    {
     inputName=VelocityRT
     input {expression=&self:MaterialExpressionTextureObjectParameter_5}
    }
    {
     inputName=NormalRT
     input {expression=&self:MaterialExpressionTextureObjectParameter_3}
    }
    {
     inputName=HeightRT
     input {expression=&self:MaterialExpressionTextureObjectParameter_2}
    }
    {
     inputName=deltaT
     input {expression=&self:MaterialExpressionCollectionParameter_15}
    }
    {
     inputName=TotalSize
     input {expression=&self:MaterialExpressionAppendVector_4}
    }
    {
     inputName=inUV
     input {expression=&self:MaterialExpressionTextureCoordinate_1}
    }
    {
     inputName=GridSize
     input {expression=&self:MaterialExpressionCollectionParameter_5}
    }
    {
     inputName=ZScale
     input {expression=&self:MaterialExpressionCollectionParameter_2}
    }
   ]
  }
  AppendVector(
   A=NamedRerouteDeclaration(None=ComponentMask(None=@n1) {g=true, r=true}) {name=Offset XY, nodeColor=<rgba 0 1 0.0625 1>}
   B=ComponentMask(None=@n1) {b=true}
  )
  n2 = Multiply(
   A=AppendVector(
    A=CollectionParameter() {collection=&MPC_Erosion, parameterName=ResX}
    B=CollectionParameter() {collection=&MPC_Erosion, parameterName=ResY}
   )
   B=CollectionParameter() {collection=&MPC_Erosion, parameterName=GridSize}
  ) {desc=RT World Size}
  n3 = Add(A=TextureCoordinate(), B=Divide(A=NamedRerouteUsage(), B=@n2)) {desc="Position, with upper left at 0,0"}
  n4 = Divide(
   A=AppendVector(
    A=CollectionParameter() {collection=&MPC_Erosion, parameterName=TotalSizeX}
    B=CollectionParameter() {collection=&MPC_Erosion, parameterName=TotalSizeY}
   ) {desc=LS World Size}
   B=@n2
  ) {desc=LS to RT Scale}
  n5 = Abs(None=Subtract(A=Divide(A=@n3, B=@n4)) {constB=0.5})
  n6 = MaterialFunctionCall(Float2=OneMinus(None=Multiply(A=@n5) {constB=2})) {materialFunction=&BreakOutFloat2Components}
  n7 = Constant() {r=1}
  NamedRerouteDeclaration(None=If(A=Min(A=@n6.R, B=@n6.R), "A > B"=@n7, "A == B"=@n7, "A < B"=Constant())) {name=Border Mask, nodeColor=<rgba 1 0.6094 0 1>}
  n8 = ComponentMask(None=@n1) {a=true}
  n9 = NamedRerouteUsage()
  NamedRerouteDeclaration(None=Saturate(None=Multiply(A=@n8, B=@n9))) {name=Water Volume, nodeColor=<rgba 0 0.3438 1 1>}
  n10 = Multiply(
   A=CollectionParameter() {collection=&MPC_Erosion, parameterName=Evaporation}
   B=CollectionParameter() {collection=&MPC_Erosion, parameterName=deltaT}
  )
  NamedRerouteDeclaration(None=OneMinus(None=Saturate(None=@n10))) {name=Fraction After Evaporation, nodeColor=<rgba 0 0.3672 1 1>}
  OneMinus(None=Multiply(A=NamedRerouteUsage(), B=NamedRerouteUsage()))
 ]
}
