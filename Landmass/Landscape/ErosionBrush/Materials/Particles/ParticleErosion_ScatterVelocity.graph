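# Unlit Landmass erosion-brush material that computes updated particle velocity and sediment from position, velocity, normal and height render targets.
# A custom HLSL node applies gravity and damping, then removes velocity into the terrain normal near the ground.
# An oscillating wind term, weighted by a direction dot product, is added.
# Further branches derive carry capacity, dissolve or deposit sediment, and build a hole mask from neighbor heights, using MPC_Erosion settings.
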Material /Landmass/Landscape/ErosionBrush/Materials/Particles/ParticleErosion_ScatterVelocity
graph {
 properties {
  bAllowNegativeEmissiveColor=true
  blendMode=BLEND_AlphaComposite
  pixelDepthOffsetMode=PDOM_Legacy
  shadingModel=MSM_Unlit
  usedShadingModels=Unlit
 }
 nodes [
  n1 = VectorParameter() {parameterName=WindDirection, defaultValue=<rgba 2 0.5 0 0>}
  n2 = ConstantBiasScale(None=DotProduct(A=Normalize(VectorInput=@n1.RGB), B=NamedRerouteUsage()))
  n3 = Multiply(
   A=Divide(A=CollectionParameter() {collection=/Landmass/Landscape/BlueprintBrushes/MPC/MPC_Erosion, parameterName=idx}) {constB=1000}
   B=ScalarParameter() {defaultValue=2, parameterName=Wind Rotation Speed}
  )
  n4 = Multiply(
   A=VectorParameter() {parameterName=WindDirection, defaultValue=<rgba 2 0.5 0 0>}.RGB
   B=ScalarParameter() {defaultValue=20, parameterName=Wind Intensity}
  )
  n5 = RotateAboutAxis(
   NormalizedRotationAxis=Constant3Vector() {constant=<rgba 0 0 1 0>}
   RotationAngle=Multiply(A=Sine(None=@n3), B=ScalarParameter() {defaultValue=0.1, parameterName=Wind Rotation})
   PivotPoint=Constant()
   Position=@n4
  )
  Add(A=Multiply(A=Saturate(None=@n2), B=Add(A=@n5, B=@n4)), B=NamedRerouteUsage())
  CollectionParameter() {collection=&MPC_Erosion, parameterName=MaxAdvectSlope}
  CollectionParameter() {collection=&MPC_Erosion, parameterName=Solubility}
  CollectionParameter() {collection=&MPC_Erosion, parameterName=ResX}
  MaterialFunctionCall() {materialFunction=/Landmass/Landscape/OldPrototype_BP/Mats/MF/PointSampleUVs}
  n6 = Constant()
  n7 = TextureSampleParameter2D!(UVs, Level=@n6, Apply View MipBias) {mipValueMode=TMVM_MipLevel, parameterName=PositionRT, texture=&Black}
  MaterialExpressionAppendVector_0 = AppendVector(A=@n7.RGB, B=@n7.RGB)
  MaterialExpressionTextureSampleParameter2D_7 = TextureSampleParameter2D(Level=@n6) {mipValueMode=TMVM_MipLevel, parameterName=VelocityRT, texture=&Black}
  MaterialExpressionTextureObjectParameter_3 = TextureObjectParameter() {parameterName=NormalRT, texture=&Black}
  MaterialExpressionTextureObjectParameter_0 = TextureObjectParameter() {parameterName=HeightRT, texture=&Black}
  MaterialExpressionCollectionParameter_3 = CollectionParameter() {collection=&MPC_Erosion, parameterName=deltaT}
  MaterialExpressionAppendVector_2 = AppendVector(
   A=CollectionParameter() {collection=&MPC_Erosion, parameterName=RTWorldSizeX}
   B=CollectionParameter() {collection=&MPC_Erosion, parameterName=RTWorldSizeY}
  )
  MaterialExpressionTextureCoordinate_1 = TextureCoordinate()
  MaterialExpressionCollectionParameter_10 = CollectionParameter() {collection=&MPC_Erosion, parameterName=VelocityDamping}
  MaterialExpressionCollectionParameter_4 = CollectionParameter() {collection=&MPC_Erosion, parameterName=ZScale}
  n8 = Custom!(
   Position=@MaterialExpressionAppendVector_0
   Velocity=@MaterialExpressionTextureSampleParameter2D_7.RGB
   NormalRT=@MaterialExpressionTextureObjectParameter_3
   HeightRT=@MaterialExpressionTextureObjectParameter_0
   deltaT=@MaterialExpressionCollectionParameter_3
   WorldSize=@MaterialExpressionAppendVector_2
   inUV=@MaterialExpressionTextureCoordinate_1
   Damping=@MaterialExpressionCollectionParameter_10
   ZScale=@MaterialExpressionCollectionParameter_4
  ) {
   code="//Position: input Texture Object containing \r\n//Velocity: input \r\n//NormalRT: input \r\n//HeightRT: input\r\n//deltaT: input\r\n//input TextureSize\r\n//input GridSize\r\n\r\nfloat3 accel = float3(0,0, -980 * deltaT);\r\n\r\n//Note: position currently converted from meters to cm due to clamping issues in RenderTargets\r\nfloat3 pos = Position * 100;\r\n\r\nfloat2 UV = inUV + ((pos ) / WorldSize);\r\n\r\nfloat startZ = HeightRT.SampleLevel(HeightRTSampler, UV, 0).b;\r\n\r\nfloat3 velocity = 0;\r\n\r\nif(Position.a > 0)\r\n{\r\n\tvelocity = Velocity + accel;\r\n\tvelocity *= exp(-Damping*deltaT);\r\n}\r\n\r\n\r\nfloat2 normal2d = NormalRT.SampleLevel(NormalRTSampler, UV, 0).rg;\r\nnormal2d = (normal2d - 0.5) * 2;\r\nfloat3 normal = normalize( float3( normal2d, sqrt(1 - dot(normal2d , normal2d ) ) ) );\r\nOutNormal  = normal;\r\n\r\n\r\nfloat terrainZ = HeightRT.SampleLevel(HeightRTSampler, UV, 0).r;\r\n\r\nif(((velocity.z  / ZScale)* deltaT) + pos.z < terrainZ + 50 &&&& Position.a > 0)\r\n{\r\n\t//float boost = saturate(-1 * (velocity.z* deltaT + pos.z -16));\r\n\tvelocity -= normal * min(0, dot(normal, velocity)) ;\r\n}\r\n\r\nreturn velocity;\r\n\r\n"
   description=Velocity Acceleration
   additionalOutputs [{outputName=OutNormal, outputType=CMOT_Float3}]
   inputs [
    {
     inputName=Position
     input {expression=&self:MaterialExpressionAppendVector_0}
    }
    {
     inputName=Velocity
     input {
      expression=&self:MaterialExpressionTextureSampleParameter2D_7
      mask=1
      maskB=1
      maskG=1
      maskR=1
     }
    }
    {
     inputName=NormalRT
     input {expression=&self:MaterialExpressionTextureObjectParameter_3}
    }
    {
     inputName=HeightRT
     input {expression=&self:MaterialExpressionTextureObjectParameter_0}
    }
    {
     inputName=deltaT
     input {expression=&self:MaterialExpressionCollectionParameter_3}
    }
    {
     inputName=WorldSize
     input {expression=&self:MaterialExpressionAppendVector_2}
    }
    {
     inputName=inUV
     input {expression=&self:MaterialExpressionTextureCoordinate_1}
    }
    {
     inputName=Damping
     input {expression=&self:MaterialExpressionCollectionParameter_10}
    }
    {
     inputName=ZScale
     input {expression=&self:MaterialExpressionCollectionParameter_4}
    }
   ]
  }
  NamedRerouteDeclaration(None=@n8.OutNormal) {name=Terrain Normal At Particle, nodeColor=<rgba 0.0054 0.2751 1 1>}
  n9 = MaterialFunctionCall() {materialFunction=&PointSampleUVs}
  n10 = Constant()
  NamedRerouteDeclaration(None=Multiply(A=TextureSampleParameter2D(UVs=@n9.TexelCenterUVs, Level=@n10) {mipValueMode=TMVM_MipLevel, parameterName=VelocityRT, texture=&Black}.A) {desc="v.A = Sediment per Water Unit"}) {name=Sedimenter Per Water Unit, nodeColor=<rgba 1 0.7031 0 1>}
  n11 = Divide(
   A=MaterialFunctionCall(Vector 3=NamedRerouteUsage()) {desc=Velocity in cm/Second, materialFunction=/Engine/Functions/Engine_MaterialFunctions02/Utility/VectorLength}.V3 Length
   B=CollectionParameter() {collection=&MPC_Erosion, parameterName=GridSize}
  ) {constB=100}
  n12 = CollectionParameter() {collection=&MPC_Erosion, parameterName=Solubility}
  NamedRerouteDeclaration(None=Multiply(A=Multiply(A=@n11, B=@n12), B=NamedRerouteUsage())) {name=Sediment Carry Capacity, nodeColor=<rgba 0 1 0.25 1>}
  NamedRerouteDeclaration(None=@n8.return) {name=Updated Velocity, nodeColor=<rgba 0.9699 1 0 1>}
  MaterialExpressionTextureObjectParameter_2 = TextureObjectParameter() {parameterName=HeightRT, texture=&Black}
  MaterialExpressionTextureCoordinate_0 = TextureCoordinate()
  MaterialExpressionAppendVector_1 = AppendVector(
   A=CollectionParameter() {collection=&MPC_Erosion, parameterName=ResX}
   B=CollectionParameter() {collection=&MPC_Erosion, parameterName=ResY}
  )
  MaterialExpressionMultiply_8 = Multiply(
   A=@MaterialExpressionAppendVector_1
   B=CollectionParameter() {collection=&MPC_Erosion, parameterName=GridSize}
  )
  n13 = MaterialFunctionCall() {materialFunction=&PointSampleUVs}
  n14 = Constant()
  MaterialExpressionAdd_0 = Add(
   A=Multiply(
    A=TextureSampleParameter2D(UVs=@n13.TexelCenterUVs, Level=@n14) {mipValueMode=TMVM_MipLevel, parameterName=VelocityRT, texture=&Black}.RGB
    B=CollectionParameter() {collection=&MPC_Erosion, parameterName=deltaT}
   )
   B=TextureSampleParameter2D(UVs=MaterialFunctionCall() {materialFunction=&PointSampleUVs}.TexelCenterUVs, Level=Constant()) {mipValueMode=TMVM_MipLevel, parameterName=PositionRT, texture=&Black}.RGB
  )
  MaterialExpressionTextureObjectParameter_1 = TextureObjectParameter() {parameterName=VelocityRT, texture=&Black}
  n15 = Custom!(
   Tex=@MaterialExpressionTextureObjectParameter_2
   UV=@MaterialExpressionTextureCoordinate_0
   TextureSize=@MaterialExpressionAppendVector_1
   WorldSize=@MaterialExpressionMultiply_8
   Position=@MaterialExpressionAdd_0
   VelocityRT=@MaterialExpressionTextureObjectParameter_1
  ) {
   code="float2 texelsize = 1/TextureSize;\r\n\r\nfloat2 offsets[8] = {float2(1,0) , float2(-1,0) , float2(0,1) , float2 (0,-1), float2(1,1), float2(1,-1), float2(-1,1), float2(-1,-1)};\r\n\r\nfloat minz = 1000000;\r\nfloat maxz = -1000000;\r\nfloat accum = 0;\r\n\r\nfloat2 startpos = UV + (Position.xy / WorldSize);\r\n\r\nfloat2 startUV = floor(startpos * TextureSize) + 0.5;\r\nstartUV /= TextureSize; \r\n\r\nfloat startZ = Tex.SampleLevel(TexSampler, startUV, 0);\r\n\r\nfor(int i = 0; i < 4; i++)\r\n{\r\n\tfloat2 curUV = startUV + (offsets[i] * texelsize);\r\n\tfloat offsetsample = Tex.SampleLevel(TexSampler, curUV, 0);\r\n\r\n\tminz = min(minz, offsetsample);\r\n\t//maxz = max(maxz, offsetsample);\r\n\taccum += offsetsample;\r\n\r\n}\r\n\r\nfloat2 velocity = normalize(VelocityRT.SampleLevel(VelocityRTSampler, startpos, 0).rg);\r\nfloat2 crossvel = float2(-velocity.y, velocity.x);\r\nfloat2 veloffsets[4] = {velocity, -velocity, crossvel, -crossvel};\r\n\r\nfor(int k = 0; k < 4; k++)\r\n{\r\n\tfloat2 curUV = startpos + (veloffsets[k] * texelsize);\r\n\tfloat offsetsample = Tex.SampleLevel(TexSampler, curUV, 0);\r\n\r\n\tmaxz = max(maxz, offsetsample);\r\n\r\n\r\n}\r\n\r\n\r\naccum /= 4;\r\n\r\nreturn float4(startZ, minz, maxz, accum);"
   description=Sample Neighbors
   outputType=CMOT_Float4
   inputs [
    {
     inputName=Tex
     input {expression=&self:MaterialExpressionTextureObjectParameter_2}
    }
    {
     inputName=UV
     input {expression=&self:MaterialExpressionTextureCoordinate_0}
    }
    {
     inputName=TextureSize
     input {expression=&self:MaterialExpressionAppendVector_1}
    }
    {
     inputName=WorldSize
     input {expression=&self:MaterialExpressionMultiply_8}
    }
    {
     inputName=Position
     input {expression=&self:MaterialExpressionAdd_0}
    }
    {
     inputName=VelocityRT
     input {expression=&self:MaterialExpressionTextureObjectParameter_1}
    }
   ]
  }
  n16 = MaterialFunctionCall(Float4=@n15) {desc="float4(startZ, minz, maxz, accum);", materialFunction=&BreakOutFloat4Components}
  n17 = Subtract(A=@n16.R, B=@n16.R) {desc=Center - Min Neighbor}
  n18 = Subtract(A=@n16.B, B=@n16.B) {desc=Max - Center}
  n19 = CollectionParameter() {collection=&MPC_Erosion, parameterName=GridSize}
  n20 = Ceil(None=Subtract(A=ScalarParameter() {defaultValue=1, parameterName=M}, B=Divide(A=@n18, B=@n19) {desc=Slope}))
  NamedRerouteDeclaration(None=Multiply(A=Saturate(None=@n17), B=Saturate(None=@n20))) {name=Hole Mask, nodeColor=<rgba 0 1 0.6484 1>}
  NamedRerouteDeclaration(None=Subtract(A=NamedRerouteUsage() {desc=How much it can carry}, B=NamedRerouteUsage() {desc=How much it is carrying}) {desc=Positive if Over-Saturated}) {name=Over Saturation, nodeColor=<rgba 0 1 0.9531 1>}
  n21 = NamedRerouteUsage()
  n22 = NamedRerouteUsage()
  n23 = ScalarParameter() {defaultValue=0.001, parameterName=Dissolving Constant}
  n24 = ScalarParameter() {defaultValue=0.1, parameterName=Deposition Constant}
  n25 = Multiply(
   A=Multiply(
    A=If(A=@n22, "A > B"=@n23, "A < B"=@n24)
    B=CollectionParameter() {collection=&MPC_Erosion, parameterName=deltaT}
   )
   B=NamedRerouteUsage()
  )
  NamedRerouteDeclaration(None=Saturate(None=Add(A=@n21, B=@n25))) {name=Updated Sediment Per Water Unit, nodeColor=<rgba 0.0825 1 0 1>}
  NamedRerouteDeclaration(None=Multiply(A=@n7.A) {constB=1000}) {name=Volume, nodeColor=<rgba 1 0.6094 0 1>}
  NamedRerouteUsage()
  OneMinus(None=NamedRerouteUsage())
 ]
}
