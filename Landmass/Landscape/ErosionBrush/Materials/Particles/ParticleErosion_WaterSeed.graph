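# Unlit alpha-composite material for Landmass erosion brushes that appears to seed water particle state, one particle per grid cell.
# Grid cells derive from the erosion MPC resolution and divisor, with time-driven vector noise supplying randomized XY start values.
# HeightRT plus a Z offset gives start height, packed with the XY noise.
# StartWater becomes inverted alpha to suit alpha composite, optionally masked to every Nth particle for debugging.
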Material /Landmass/Landscape/ErosionBrush/Materials/Particles/ParticleErosion_WaterSeed
graph {
 properties {
  bAllowNegativeEmissiveColor=true
  bUsedWithInstancedStaticMeshes=true
  blendMode=BLEND_AlphaComposite
  pixelDepthOffsetMode=PDOM_Legacy
  shadingModel=MSM_Unlit
  usedShadingModels=Unlit
 }
 nodes [
  n1 = Multiply(
   A=CollectionParameter() {collection=/Landmass/Landscape/BlueprintBrushes/MPC/MPC_Erosion, parameterName=idx}
   B=ScalarParameter() {defaultValue=1.618, parameterName=TS}
  )
  n2 = Divide(
   A=AppendVector(
    A=CollectionParameter() {collection=&MPC_Erosion, parameterName=ResX}
    B=CollectionParameter() {collection=&MPC_Erosion, parameterName=ResY}
   )
   B=CollectionParameter() {collection=&MPC_Erosion, parameterName=ResolutionDivisor}
  ) {constB=1}
  n3 = Floor(None=Multiply(A=TextureCoordinate(), B=@n2))
  n4 = Multiply(A=Time(), B=ScalarParameter() {defaultValue=1.618, parameterName=TS})
  n5 = Multiply(A=VectorNoise(World Position=AppendVector(A=@n3, B=@n4)), B=ScalarParameter() {defaultValue=0.8, parameterName=StartNoise})
  AppendVector(A=Multiply(A=Noise!(World Position=@n1, FilterWidth), B=ScalarParameter() {defaultValue=0.8, parameterName=StartNoise}), B=@n5)
  n6 = TextureSampleParameter2D!(UVs, Level=Constant(), Apply View MipBias) {mipValueMode=TMVM_MipLevel, parameterName=HeightRT, texture=&Black}
  n7 = ScalarParameter() {defaultValue=100, desc=How high above Terrain to Start, parameterName=Water Start Z Offset}
  AppendVector(A=ComponentMask(None=@n5) {g=true, r=true}, B=Divide(A=Add(A=@n6.R, B=@n7)) {constB=100})
  n8 = CollectionParameter() {collection=&MPC_Erosion, parameterName=ResX}
  n9 = CollectionParameter() {collection=&MPC_Erosion, parameterName=ResolutionDivisor}
  n10 = Floor(None=Multiply(A=TextureCoordinate(), B=Divide(A=@n8, B=@n9)))
  n11 = VectorNoise(World Position=AppendVector(A=@n10, B=Constant()))
  Ceil(None=Subtract(A=ComponentMask(None=@n11) {r=true}, B=ScalarParameter() {parameterName=RandomThreshold}) {constB=0.99})
  CollectionParameter() {collection=&MPC_Erosion, parameterName=idx}
  Constant2Vector()
  Divide(A=Add(A=@n3) {constB=0.5}, B=@n2)
  MaterialFunctionCall() {materialFunction=/Landmass/Landscape/OldPrototype_BP/Mats/MF/PointSampleUVs}
  n12 = Fmod(A=@n10, B=ScalarParameter() {defaultValue=4, parameterName=Show Every N Particles})
  n13 = MaterialFunctionCall(Float2=OneMinus(None=Saturate(None=@n12))) {materialFunction=&BreakOutFloat2Components}
  NamedRerouteDeclaration(None=Saturate(None=Min(A=@n13.R, B=@n13.R))) {name=Isolate Particle Debug, nodeColor=<rgba 0 0.8594 1 1>}
  Noise(World Position=Add(A=@n4) {constB=1619})
  n14 = Divide(A=CollectionParameter() {collection=&MPC_Erosion, parameterName=StartWater}) {constB=1000, desc=Max Water}
  n15 = NamedRerouteUsage()
  n16 = ScalarParameter() {parameterName=Isolate Particles Debug}
  OneMinus(None=Multiply(A=Saturate(None=@n14), B=LinearInterpolate(B=@n15, Alpha=@n16) {constA=1})) {desc=Alpha Compoite writes Inverse Alpha}
 ]
}
