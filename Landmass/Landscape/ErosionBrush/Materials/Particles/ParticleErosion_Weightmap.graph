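# Additive unlit material for the Landmass erosion brush that computes a weightmap value from particle erosion simulation data.
# It point-samples SceneCaptureRT at texel centers, unpacking water velocity, sediment transfer and water volume into named reroutes.
# Velocity magnitude is scaled by resolution and timestep from MPC_Erosion, while sediment transfer over grid size splits by sign into deposition and erosion.
# The Velocity, Deposition and Erosion parameters weight these terms before they are summed.
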Material /Landmass/Landscape/ErosionBrush/Materials/Particles/ParticleErosion_Weightmap
graph {
 properties {
  bAllowNegativeEmissiveColor=true
  blendMode=BLEND_Additive
  pixelDepthOffsetMode=PDOM_Legacy
  shadingModel=MSM_Unlit
  usedShadingModels=Unlit
 }
 nodes [
  n1 = Divide(
   A=MaterialFunctionCall(Vector 2=NamedRerouteUsage()) {materialFunction=/Engine/Functions/Engine_MaterialFunctions02/Utility/VectorLength}.V2 Length
   B=CollectionParameter() {collection=/Landmass/Landscape/BlueprintBrushes/MPC/MPC_Erosion, parameterName=ResX}
  )
  n2 = CollectionParameter() {collection=&MPC_Erosion, parameterName=deltaT}
  n3 = Divide(A=NamedRerouteUsage(), B=CollectionParameter() {collection=&MPC_Erosion, parameterName=GridSize})
  n4 = Max(A=@n3) {constB=0, desc="Positive = Add soil to terrain"}
  n5 = ScalarParameter() {parameterName=Deposition}
  n6 = Max(A=Multiply(A=@n3) {constB=-1}) {constB=0, desc="Negative = Erode"}
  n7 = ScalarParameter() {parameterName=Erosion}
  Add(
   A=Multiply(A=Multiply(A=@n1, B=@n2), B=ScalarParameter() {parameterName=Velocity})
   B=Add(A=Multiply(A=@n4, B=@n5), B=Multiply(A=@n6, B=@n7))
  ) {constB=0}
  n8 = TextureSampleParameter2D(UVs=MaterialFunctionCall() {materialFunction=/Landmass/Landscape/OldPrototype_BP/Mats/MF/PointSampleUVs}.TexelCenterUVs) {parameterName=SceneCaptureRT, texture=&Black}
  NamedRerouteDeclaration(None=ComponentMask(None=@n8.RGB) {g=true, r=true}) {name=Water Velocity, nodeColor=<rgba 0.875 0 1 1>}
  NamedRerouteDeclaration(None=@n8.B) {name=Sediment Transfer, nodeColor=<rgba 1 0.2344 0 1>}
  NamedRerouteDeclaration(None=Multiply(A=Multiply(A=@n8.A) {constB=100, desc=Un-Normalized For Particle Overlap}) {constB=1000, desc=Un-Normalize for MaxWater Per Particle}) {name=Water Volume, nodeColor=<rgba 0 0.4439 1 1>}
 ]
}
